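# Material function for the Teriyaki Fish Toon character that layers a tiled, masked sequin effect over incoming material attributes.
# Sequin color blends two falloff tints by Fresnel and tile pattern, optionally multiplied by CerealBoxBaseColor.
# A sine of the camera-space local up vector plus the tile texture adds faint view-dependent glints.
# The channel-selected mask, defaulting to vertex red, blends the result into emissive, metallic and normal, with switchable base normals.
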MaterialFunction /BRCosmetics/Characters/Player/Male/Medium/Bodies/M_MED_Teriyaki_Fish_Toon/Materials/MF_TeriyakiFishToon_Custom1
graph {nodes [
  n1 = FunctionInput(Preview=SetMaterialAttributes!(MaterialAttributes)) {bUsePreviewValueAsDefault=true, inputName=inMat, inputType=FunctionInput_MaterialAttributes}
  n2 = GetMaterialAttributes(None=@n1) {attributeGetTypes [
    #1
    #2
    #3
    #4
   ]}
  n3 = GetMaterialAttributes(None=@n1) {attributeGetTypes [
    #3
    #4
    #1
    #2
   ]}
  n4 = Saturate(None=Transform(None=Constant3Vector() {constant=<rgba 0 1 0 0>}) {transformSourceType=TRANSFORMSOURCE_Local, transformType=TRANSFORM_Camera})
  n5 = ComponentMask(None=@n4) {b=true}
  n6 = Add(A=ComponentMask(None=@n4) {r=true}, B=ComponentMask(None=@n4) {g=true})
  n7 = NamedRerouteUsage()
  n8 = Sine(None=Add(
    A=Add(A=@n5, B=@n6)
    B=TextureSample(UVs=@n7) {samplerType=SAMPLERTYPE_LinearGrayscale, texture=/Game/Global/Textures/Default/Tileable_Patterns/T_SequinTile}.RGB
   ))
  n9 = Multiply(A=OneMinus(None=Saturate(None=@n8))) {constB=0.05}
  n10 = Power(Base=Fresnel() {exponent=1})
  n11 = OneMinus(None=@n10)
  n12 = Multiply(A=@n11, B=@n10)
  n13 = Multiply(
   A=ScalarParameter() {defaultValue=1.5, group=MF_Sequin, parameterName=SequinBrightness}
   B=Multiply(A=NamedRerouteUsage(), B=LinearInterpolate(Alpha=@n12) {constA=0.7})
  )
  n14 = VertexColor()
  n15 = Constant()
  n16 = Constant() {r=0.01}
  n17 = ChannelMaskParameter(None=FunctionInput(Preview=SphereMask(A=@n14.R, B=@n15, Radius=@n16) {attenuationRadius=0.05}) {inputName=SequinMask, inputType=FunctionInput_Scalar}) {group=MF_Sequin, parameterName=SequinMaskChannel}
  n18 = ScalarParameter() {defaultValue=0.9, group=MF_Sequin, parameterName=SequinMetalness}
  n19 = NamedRerouteUsage()
  n20 = ScalarParameter() {group=MF_Sequin, parameterName=UseBaseMetalness}
  n21 = NamedRerouteUsage()
  n22 = TextureSampleParameter2D(UVs=NamedRerouteUsage()) {
   group=MF_Sequin
   parameterName=Sequin_TilingNormal
   samplerType=SAMPLERTYPE_Normal
   texture=/Game/Global/Textures/Default/Tileable_Patterns/T_SequinTile_N
  }
  n23 = Multiply(A=ScalarParameter() {defaultValue=0.8, group=MF_Sequin, parameterName=SequinNormalIntensity}, B=Constant3Vector() {constant=<rgba 1 1 0 0>})
  n24 = Add(A=@n21, B=Multiply(A=@n22.RGB, B=@n23)) {constB=0}
  n25 = SetMaterialAttributes!(
   MaterialAttributes=@n2.MaterialAttributes
   EmissiveColor=LinearInterpolate(A=@n3.EmissiveColor, B=Add(A=@n9, B=@n13), Alpha=@n17.R)
   Metallic=LinearInterpolate(A=@n3.Metallic, B=LinearInterpolate(A=@n18, B=@n19, Alpha=@n20), Alpha=@n17.R)
   Roughness
   Normal=LinearInterpolate(A=@n3.Normal, B=StaticSwitchParameter(True=@n21, False=@n24) {group=MF_Sequin, parameterName=UseBaseNormal}, Alpha=@n17.R)
  ) {attributeSetTypes [
    #3
    #4
    #1
    #2
   ]}
  FunctionOutput(None=@n25)
  Multiply(A=@n21, B=@n23)
  NamedRerouteDeclaration(None=Multiply(A=TextureCoordinate(), B=ScalarParameter() {defaultValue=32, group=MF_Sequin, parameterName=SequinTile})) {name=SequinUVs, nodeColor=<rgba 0 1 0.0859 1>}
  n26 = TextureSampleParameter2D(UVs=NamedRerouteUsage()) {
   group=MF_Sequin
   parameterName=Sequin_TilingOffsetPattern
   samplerType=SAMPLERTYPE_LinearGrayscale
   texture=&T_SequinTile
  }
  NamedRerouteDeclaration(None=@n26.R) {name=SequinOffset, nodeColor=<rgba 1 0 0.6562 1>}
  n27 = VectorParameter() {group=MF_Sequin, parameterName=SequinFalloffColor01, defaultValue=<rgba 0.8842 0.6282 1 1>}
  n28 = VectorParameter() {group=MF_Sequin, parameterName=SequinFalloffColor02, defaultValue=<rgba 0.6663 0.575 1 1>}
  n29 = Multiply(A=LinearInterpolate(Alpha=NamedRerouteUsage()) {constA=0.5, constB=1.5}, B=@n11)
  n30 = VectorParameter() {parameterName=CerealBoxBaseColor, defaultValue=<rgba 0.45 0.5408 0.6042 1>}
  n31 = Multiply(A=@n30.RGB, B=@n27.RGB)
  n32 = Multiply(A=@n30.RGB, B=@n28.RGB)
  NamedRerouteDeclaration(None=LinearInterpolate(
    A=LinearInterpolate(A=@n27.RGB, B=@n28.RGB, Alpha=@n29)
    B=LinearInterpolate(A=@n31, B=@n32, Alpha=@n29)
    Alpha=ScalarParameter() {defaultValue=1, group=MF_Sequin, parameterName=UseBaseColor}
   )) {name=SequinColor, nodeColor=<rgba 0.8984 0 1 1>}
  NamedRerouteDeclaration(None=@n2.Roughness) {name=BaseRoughness, nodeColor=<rgba 0.6172 1 0 1>}
  NamedRerouteDeclaration(None=@n2.Normal) {name=BaseNormals, nodeColor=<rgba 0 1 0.1562 1>}
  NamedRerouteDeclaration(None=@n2.EmissiveColor) {name=BaseColor, nodeColor=<rgba 0 1 0.7188 1>}
  NamedRerouteDeclaration(None=@n2.Metallic) {name=BaseMEtalness, nodeColor=<rgba 0.6406 0 1 1>}
  NamedRerouteUsage()
 ]}
