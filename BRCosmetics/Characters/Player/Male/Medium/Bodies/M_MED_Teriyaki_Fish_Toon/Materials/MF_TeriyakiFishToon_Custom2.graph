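# Material function that recolors the emissive of selected hair regions with two division colors, for the Teriyaki Fish Toon character.
# Two narrow SphereMasks on vertex color red, near 0.7 and 0.75, sum into a blend mask over the incoming emissive.
# A Step on pre-skinned local X picks HairColor_Division1 or HairColor_Division2.
# The position is divided by bounds Max minus itself, a zero divisor rather than the bounds size.
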MaterialFunction /BRCosmetics/Characters/Player/Male/Medium/Bodies/M_MED_Teriyaki_Fish_Toon/Materials/MF_TeriyakiFishToon_Custom2
graph {nodes [
  n1 = GetMaterialAttributes(None=FunctionInput(Preview=SetMaterialAttributes!(MaterialAttributes)) {bUsePreviewValueAsDefault=true, inputName=inMat, inputType=FunctionInput_MaterialAttributes}) {attributeGetTypes [#1]}
  n2 = PreSkinnedLocalBounds()
  n3 = Divide(A=VertexInterpolator(VS=PreSkinnedPosition()).PS, B=Subtract(A=@n2.Max, B=@n2.Max))
  n4 = LinearInterpolate(
   A=VectorParameter() {parameterName=HairColor_Division2, defaultValue=<rgba 0.0208 0.0208 0.0208 1>}.RGB
   B=VectorParameter() {parameterName=HairColor_Division1, defaultValue=<rgba 0.2865 0.0113 0.0542 1>}.RGB
   Alpha=Step(Y=ComponentMask(None=@n3) {r=true}) {constX=0}
  )
  n5 = VertexColor()
  n6 = Add(
   A=SphereMask(A=@n5.R, B=Constant() {r=0.75}, Radius=Constant() {r=0.01}) {attenuationRadius=0.05}
   B=SphereMask(A=@n5.R, B=Constant() {r=0.7}, Radius=Constant() {r=0.01}) {attenuationRadius=0.05}
  )
  FunctionOutput(None=SetMaterialAttributes!(MaterialAttributes=@n1.MaterialAttributes, EmissiveColor=LinearInterpolate(A=@n1.EmissiveColor, B=@n4, Alpha=@n6)) {attributeSetTypes [#1]})
 ]}
