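# Material function for the Teriyaki Fish Toon character body that layers a SnowHeart emissive onto incoming material attributes.
# A tintable SnowHeart emissive texture is added to the incoming emissive color, and the result is blended in through a vertex color mask.
# The mask sums narrow SphereMasks used as value matchers, selecting two specific green channel levels and near-zero red.
# All other material attributes pass through unchanged.
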MaterialFunction /BRCosmetics/Characters/Player/Male/Medium/Bodies/M_MED_Teriyaki_Fish_Toon/Materials/MF_TeriyakiFishToon_SnowHeart
graph {nodes [
  n1 = GetMaterialAttributes(None=FunctionInput(Preview=SetMaterialAttributes!(MaterialAttributes)) {bUsePreviewValueAsDefault=true, inputName=inMat, inputType=FunctionInput_MaterialAttributes}) {attributeGetTypes [#1]}
  n2 = TextureSampleParameter2D() {
   parameterName=Emissive_SnowHeart
   samplerType=SAMPLERTYPE_LinearColor
   texture=/BRCosmetics/Characters/Player/Male/Medium/Bodies/M_MED_Teriyaki_Fish_Toon/Textures/T_M_MED_Teriyaki_Fish_Toon_SnowHeartEms
  }
  n3 = Add(A=@n1.EmissiveColor, B=Multiply(A=@n2.RGB, B=VectorParameter() {parameterName=EmissiveMult_SnowHeart, defaultValue=<rgba 1 1 1 1>}.RGB))
  n4 = VertexColor()
  n5 = Constant() {r=0.1}
  n6 = Constant() {r=0.01}
  n7 = VertexColor()
  n8 = Constant() {r=0.2}
  n9 = Constant() {r=0.01}
  n10 = Add(
   A=Add(A=SphereMask(A=@n4.G, B=@n5, Radius=@n6) {attenuationRadius=0.05}, B=SphereMask(A=@n7.G, B=@n8, Radius=@n9) {attenuationRadius=0.05})
   B=SphereMask(A=VertexColor().R, B=Constant(), Radius=Constant() {r=0.01}) {attenuationRadius=0.05}
  )
  FunctionOutput(None=SetMaterialAttributes!(MaterialAttributes=@n1.MaterialAttributes, EmissiveColor=LinearInterpolate(A=@n1.EmissiveColor, B=@n3, Alpha=@n10)) {attributeSetTypes [#1]})
 ]}
