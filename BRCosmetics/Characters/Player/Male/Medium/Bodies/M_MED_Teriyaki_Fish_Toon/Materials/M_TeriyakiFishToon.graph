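# Toon character material for the Teriyaki Fish Toon body, drawing cel-shaded color through emissive on a Cloth shading model.
# Vertex color indexes a color atlas row chosen by toon_index, or masks custom per-part color parameters.
# Sun-direction shading and distance-field ink lines with camera-distance thickness modulate the color, with separate mobile paths.
# Static switches layer variant functions such as CerealBox and Cube Queen before LUT and shared character PostFX.
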Material /BRCosmetics/Characters/Player/Male/Medium/Bodies/M_MED_Teriyaki_Fish_Toon/Materials/M_TeriyakiFishToon
graph {
 properties {
  bAutomaticallySetUsageInEditor=false
  bUseMaterialAttributes=true
  bUsedWithMeshDeformer=true
  bUsedWithMorphTargets=true
  bUsedWithSkeletalMesh=true
  pixelDepthOffsetMode=PDOM_Legacy
  shadingModel=MSM_Cloth
  usedShadingModels=Cloth
 }
 nodes [
  Constant3Vector() {constant=<rgba 0.02 0.02 0.02 0>}
  Constant() {r=1}
  Constant()
  LinearInterpolate()
  n1 = VertexColor()
  n2 = TextureObject() {
   samplerType=SAMPLERTYPE_LinearColor
   texture=&/TeriyakiFishToon_ColorAtlas
  }
  n3 = MaterialFunctionCall(Float2=TextureProperty(Tex=@n2)) {materialFunction=&BreakOutFloat2Components}
  n4 = Multiply(A=ScalarParameter() {defaultValue=0.5, parameterName=Index_Offset_X}, B=Divide(B=@n3.R))
  n5 = ScalarParameter() {group=TeriyakiToon, parameterName=toon_index}
  n6 = ScalarParameter() {defaultValue=1, parameterName=UsePreset}
  n7 = Multiply(A=Divide(B=@n3.G), B=Add(A=Multiply(A=@n5, B=@n6), B=ScalarParameter() {defaultValue=0.5, parameterName=Index_Offset_Y}) {constB=0.5})
  n8 = TextureSample(UVs=MaterialFunctionCall(X=Add(A=@n1, B=@n4) {constB=0.0125}, Y=@n7) {materialFunction=&MakeFloat2}.Result) {
   samplerType=SAMPLERTYPE_LinearColor
   texture=&TeriyakiFishToon_ColorAtlas
  }
  n9 = ScalarParameter() {parameterName=UseCustomColors}
  n10 = LinearInterpolate(A=@n8.RGB, B=NamedRerouteUsage(), Alpha=@n9)
  n11 = Multiply(
   A=@n10
   B=LinearInterpolate(
    A=ScalarParameter() {defaultValue=0.5, parameterName=EmissiveMult_Presets}
    B=ScalarParameter() {defaultValue=0.75, parameterName=EmissiveMult_CustomStyle}
    Alpha=@n9
   ) {constA=0.5, constB=0.75}
  ) {constB=0.5}
  n12 = VertexInterpolator(VS=PreSkinnedPosition())
  n13 = PreSkinnedLocalBounds()
  n14 = Subtract(A=@n13.Max, B=@n13.Max)
  n15 = ConstantBiasScale(None=ComponentMask(None=Divide(A=@n12.PS, B=@n14)) {r=true}) {bias=0.05, scale=0.6}
  n16 = Constant() {r=25}
  n17 = Floor(None=Multiply(A=@n15, B=@n16) {constB=10})
  n18 = MaterialFunctionCall(Hue Shift Percentage=Multiply(A=Divide(A=@n17, B=@n16)) {constB=5}, Texture=Constant3Vector() {constant=<rgba 1 0 0 0>}) {materialFunction=&HueShift}
  n19 = NamedRerouteUsage()
  n20 = Constant() {r=0.01}
  n21 = SphereMask(A=@n19, B=Multiply(B=Divide(B=@n3.R)) {constA=30, constB=0.0125}, Radius=@n20)
  n22 = Constant() {r=15}
  n23 = Floor(None=Multiply(A=@n15, B=@n22) {constB=10})
  n24 = MaterialFunctionCall(Hue Shift Percentage=Multiply(A=Divide(A=@n23, B=@n22)) {constB=6}, Texture=Constant3Vector() {constant=<rgba 1 1 0 0>}) {materialFunction=&HueShift}
  n25 = Constant3Vector() {constant=<rgba 0.1496 0.026 1 0>}
  n26 = SphereMask(A=@n24.Result, B=Constant3Vector() {constant=<rgba 0.0612 0.7305 1 1>}, Radius=Constant() {r=0.3})
  n27 = Constant() {r=0.7}
  n28 = LinearInterpolate(
   A=@n11
   B=LinearInterpolate(
    A=LinearInterpolate(A=@n11, B=@n18.Result, Alpha=@n21)
    B=LinearInterpolate(A=@n24.Result, B=@n25, Alpha=@n26)
    Alpha=SphereMask(A=@n19, B=@n27, Radius=@n20)
   )
   Alpha=ScalarParameter() {parameterName=EnableRainbowTuft}
  )
  n29 = ScalarParameter() {defaultValue=0.1, parameterName=ShadingSmooth}
  n30 = VectorParameter() {parameterName=Custom Lighting Vector, defaultValue=<rgba 0 0 1 0>}
  n31 = CollectionParameter() {
   collection=&FortniteMaterialParameters
   parameterName=SunlightDirectionalVector
  }
  n32 = DotProduct(
   A=ComponentMask(None=StaticSwitchParameter(True=@n30.RGB, False=@n31) {parameterName=Use Custom Lighting Vector}) {b=true, g=true, r=true}
   B=PixelNormalWS()
  )
  n33 = LinearInterpolate(
   A=Multiply(A=@n28, B=SmoothStep(Max=@n29, Value=@n32) {constMax=0})
   B=@n28
   Alpha=ScalarParameter() {defaultValue=0.45, parameterName=ShadingDarkenAmount}
  )
  n34 = ScalarParameter() {defaultValue=1, parameterName=EnableShading}
  n35 = Constant() {r=0.35}
  n36 = ScalarParameter() {parameterName=Merry Marauder}
  n37 = NamedRerouteUsage()
  n38 = Constant() {r=0.95}
  n39 = Constant() {r=0.01}
  n40 = TextureSampleParameter2D() {
   parameterName=DistanceField_InkLines
   samplerType=SAMPLERTYPE_Grayscale
   texture=/Game/Global/Textures/Default/Blanks/T_BlackLinearGrayscale_sRGB
  }
  n41 = TextureSampleParameter2D() {
   parameterName=DistanceField_InkLines_Mobile
   samplerType=SAMPLERTYPE_LinearGrayscale
   texture=/BRCosmetics/Characters/Player/Male/Medium/Bodies/M_MED_Teriyaki_Fish_Toon/Textures/T_M_MED_Teriyaki_Fish_Toon_Head_DF_Mobile
  }
  n42 = LinearInterpolate(
   B=MaterialFunctionCall!(Crop Right=Constant(), Crop Bottom=Constant(), Crop Top=Constant() {r=0.75}, Crop Left=Constant() {r=0.83}) {materialFunction=/Engine/Functions/Engine_MaterialFunctions01/Texturing/UVCropping}.Result
   Alpha=NamedRerouteUsage()
  )
  n43 = Constant() {r=0.17}
  n44 = Constant()
  n45 = Constant() {r=0.75}
  n46 = Constant() {r=0.77}
  n47 = LinearInterpolate(
   A=@n42
   B=Add(
    A=@n42
    B=MaterialFunctionCall(Crop Right=@n43, Crop Bottom=@n44, Crop Top=@n45, Crop Left=@n46) {materialFunction=&UVCropping}.Result
   )
   Alpha=NamedRerouteUsage()
  ) {constA=1}
  n48 = Multiply(A=@n41.RGB, B=OneMinus(None=Saturate(None=@n47)))
  n49 = ChannelMaskParameter(None=ShadingPathSwitch(Default=@n40.RGB, Deferred=@n40.RGB, Mobile=@n48) {desc=Migrated from FeatureLevelSwitch}) {parameterName=DistanceField_Channel}
  n50 = Multiply(
   A=Power(Base=@n49.R, Exp=ScalarParameter() {defaultValue=9, parameterName=DF_BlurPow}) {constExponent=3}
   B=ScalarParameter() {defaultValue=1, parameterName=DF_BlurMult}
  ) {constB=0.5}
  n51 = ScalarParameter() {defaultValue=0.05, group=NeonDepth, parameterName=DF_LineThickness_Near}
  n52 = ShadingPathSwitch(Default=@n51, Deferred=@n51, Mobile=ScalarParameter() {defaultValue=0.05, group=NeonDepth, parameterName=DF_LineThickness_Near_Mobile}) {desc=Migrated from FeatureLevelSwitch}
  n53 = ScalarParameter() {defaultValue=0.15, group=NeonDepth, parameterName=DF_LineThickness_Far}
  n54 = ShadingPathSwitch(Default=@n53, Deferred=@n53, Mobile=ScalarParameter() {defaultValue=0.15, group=NeonDepth, parameterName=DF_LineThickness_Far_Mobile}) {desc=Migrated from FeatureLevelSwitch}
  n55 = Distance(A=WorldPosition().XYZ, B=CameraPositionWS())
  n56 = ScalarParameter() {defaultValue=100, group=NeonDepth, parameterName=DF_FarDistance}
  n57 = ScalarParameter() {defaultValue=500, group=NeonDepth, parameterName=DF_NearDistance}
  n58 = Saturate(None=Divide(A=Subtract(A=@n55, B=@n56), B=Subtract(A=@n57, B=@n56)))
  n59 = Saturate(None=SphereMask(
    A=@n49.R
    B=Constant() {r=1}
    Radius=LinearInterpolate(A=@n52, B=@n54, Alpha=@n58)
    Hardness=ScalarParameter() {defaultValue=0.8, parameterName=DF_LineHardness}
   ) {attenuationRadius=0.1})
  n60 = LinearInterpolate(A=Constant3Vector(), B=Constant3Vector() {constant=<rgba 1 1 1 0>}, Alpha=Multiply(A=OneMinus(None=@n50), B=OneMinus(None=@n59)))
  n61 = Constant() {r=1}
  n62 = ScalarParameter() {defaultValue=-4, group=HoloFoil, parameterName=BL_OutlineC1}
  n63 = ScalarParameter() {defaultValue=12, group=HoloFoil, parameterName=BL_OutlineC2}
  n64 = ScalarParameter() {defaultValue=6, group=HoloFoil, parameterName=BL_OutlineExponent1}
  n65 = PixelNormalWS()
  n66 = PixelDepth()
  n67 = ScalarParameter() {defaultValue=0.01, parameterName=BL_OutlineExpMult}
  n68 = Power(
   Base=OneMinus(None=Fresnel(ExponentIn=@n64, Normal=@n65) {baseReflectFraction=0})
   Exp=Multiply(A=ScalarParameter() {defaultValue=12, group=HoloFoil, parameterName=BL_OutlineExponent2}, B=Multiply(A=@n66.R, B=@n67) {constB=0.002})
  )
  n69 = LinearInterpolate(
   B=Saturate(None=LinearInterpolate(A=@n62, B=@n63, Alpha=@n68))
   Alpha=ScalarParameter() {group=HoloFoil, parameterName="UseFresnelLines (Mobile)"}
  ) {constA=1}
  n70 = Multiply(
   A=LinearInterpolate(
    A=LinearInterpolate(A=@n11, B=@n33, Alpha=@n34)
    B=LinearInterpolate(B=@n35, Alpha=@n36)
    Alpha=SphereMask(A=@n37, B=@n38, Radius=@n39) {attenuationRadius=0.05}
   ) {constB=0}
   B=Multiply(
    A=NamedRerouteDeclaration(None=@n60) {name=DF Line Preview, nodeColor=<rgba 0 1 0.0625 1>}
    B=ShadingPathSwitch(Default=@n61, Deferred=@n61, Mobile=@n69) {desc=Migrated from FeatureLevelSwitch}
   )
  )
  MaterialFunctionCall!(Color=@n70) {materialFunction=/Game/Packages/Fortress_SharedMaterials/Anime_Master_Material/MF_TODColorCurve}
  MaterialFunctionCall() {materialFunction=&Color_Time_of_Day}
  n71 = MaterialFunctionCall() {materialFunction=&BaseFN_CharacterFunction_PreFX}
  n72 = VertexColor()
  n73 = Constant()
  n74 = Constant() {r=0.01}
  n75 = Constant()
  n76 = VertexColor()
  n77 = Constant() {r=0.001}
  n78 = SetMaterialAttributes!(
   MaterialAttributes=@n71.Result
   Base Color=Multiply(A=Step(X=@n10) {constX=0.1, constY=0.02}) {constB=0.0001}
   Metallic=@n75
   Specular=@n75
   Roughness=LinearInterpolate(A=@n76.R, B=@n76.R, Alpha=NamedRerouteUsage())
   Fuzz Color=MaterialFunctionCall(Y=@n77, Z=@n77) {materialFunction=&MakeFloat3}.Result
   Emissive Color=MaterialFunctionCall!(Color=Saturate(None=@n70)) {materialFunction=/Game/Packages/Fortress_SharedMaterials/Anime_Master_Material/MF_TODColorAdjustment}.Result
   Ambient Occlusion=NamedRerouteUsage()
   Cloth=Constant() {desc=Opacity, r=1}
  ) {attributeSetTypes [
    #1
    #2
    #3
    #4
    #5
    #6
    #7
    #8
   ]}
  n79 = StaticSwitchParameter(
   True=MaterialFunctionCall!(SequinMask=SphereMask(A=@n72.R, B=@n73, Radius=@n74) {attenuationRadius=0.05}, inMat=@n78) {materialFunction=&/MF_TeriyakiFishToon_Custom1}.Result
   False=@n78
  ) {parameterName=CerealBox}
  n80 = StaticSwitchParameter(
   True=MaterialFunctionCall!(inMat=@n79) {materialFunction=&/MF_TeriyakiFishToon_SnowHeart}.Result
   False=@n79
  ) {parameterName=SnowHeart}
  n81 = StaticSwitchParameter(
   True=MaterialFunctionCall!(inMat=@n80) {materialFunction=&/MF_TeriyakiFishToon_Custom2}.Result
   False=@n80
  ) {parameterName=Division}
  n82 = StaticSwitchParameter(
   True=MaterialFunctionCall!(inMat=@n81) {materialFunction=&/MF_TeriyakiFishToon_DarkBomber}.Result
   False=@n81
  ) {parameterName=Dark Bomber}
  n83 = GetMaterialAttributes(None=StaticSwitchParameter(
    True=MaterialFunctionCall!(inMat=@n82) {materialFunction=&/MF_TeriyakiFishToon_CubeQueen}.Result
    False=@n82
   ) {parameterName=Cube Queen}) {attributeGetTypes [#6]}
  MaterialFunctionCall(
   AO=Constant() {r=1}
   Distance Blend=@n71.Distance Blend
   Fresnel=@n71.Distance Blend
   Gmap=@n71.Distance Blend
   Pre Skinned Local Position=@n71.Distance Blend
   Material=SetMaterialAttributes!(
    MaterialAttributes=@n83.MaterialAttributes
    Emissive Color=MaterialFunctionCall!(In=@n83.Emissive Color, LUT, LUT Multiplier) {materialFunction=/Game/Packages/Fortress_SharedMaterials/Anime_Master_Material/MF_ApplyLUT}.Result
   ) {attributeSetTypes [#6]}
  ) {materialFunction=&BaseFN_CharacterFunction_PostFX}
  MaterialFunctionCall(Float4=TextureSampleParameter2D() {parameterName=SSC Texture, texture=&T_BlacksRGB}.RGB) {materialFunction=&BreakOutFloat4Components}
  Multiply()
  n84 = NamedRerouteUsage()
  n85 = Constant()
  n86 = LinearInterpolate(
   A=LinearInterpolate(
    A=Constant3Vector() {constant=<rgba 0.9 0.9 0.9 0>}
    B=VectorParameter() {parameterName=SkinColor, defaultValue=<rgba 0.9 0.9 0.9 0>}.RGB
    Alpha=SphereMask(A=@n84, B=@n85) {attenuationRadius=0.05}
   )
   B=VectorParameter() {parameterName=ArmsColor, defaultValue=<rgba 0.9 0.9 0.9 0>}.RGB
   Alpha=SphereMask(A=NamedRerouteUsage(), B=Constant() {r=0.1}) {attenuationRadius=0.05}
  )
  n87 = VectorParameter() {parameterName=LegsColor, defaultValue=<rgba 0.9 0.9 0.9 0>}
  n88 = SphereMask(A=NamedRerouteUsage(), B=Constant() {r=0.2}) {attenuationRadius=0.05}
  n89 = SphereMask(A=NamedRerouteUsage(), B=Constant() {r=0.3}) {attenuationRadius=0.05}
  n90 = SphereMask(A=NamedRerouteUsage(), B=Constant() {r=1}) {attenuationRadius=0.05}
  n91 = LinearInterpolate(
   A=LinearInterpolate(
    A=LinearInterpolate(A=@n86, B=@n87.RGB, Alpha=@n88)
    B=VectorParameter() {parameterName=AccessoriesColor, defaultValue=<rgba 0.9 0.9 0.9 0>}.RGB
    Alpha=Add(A=@n89, B=@n90)
   )
   B=VectorParameter() {parameterName=ShoesColor, defaultValue=<rgba 0.9 0.9 0.9 0>}.RGB
   Alpha=SphereMask(A=NamedRerouteUsage(), B=Constant() {r=0.4}) {attenuationRadius=0.05}
  )
  n92 = VectorParameter() {parameterName=HatColor}
  n93 = Add(A=Add(A=@n92.R, B=@n92.R), B=@n92.B)
  n94 = LinearInterpolate(A=NamedRerouteUsage(), B=@n92.RGB, Alpha=Saturate(None=Ceil(None=@n93)))
  n95 = SphereMask(A=NamedRerouteUsage(), B=Constant() {r=0.5}) {attenuationRadius=0.05}
  n96 = NamedRerouteUsage()
  n97 = Constant() {r=0.6}
  n98 = LinearInterpolate(
   A=LinearInterpolate(
    A=LinearInterpolate(A=@n91, B=@n94, Alpha=@n95)
    B=VectorParameter() {parameterName=GloveColor, defaultValue=<rgba 0.9047 0.9047 0.9047 0>}.RGB
    Alpha=SphereMask(A=@n96, B=@n97) {attenuationRadius=0.05}
   )
   B=VectorParameter() {parameterName=ScarfColor, defaultValue=<rgba 0.9047 0.9047 0.9047 0>}.RGB
   Alpha=SphereMask(A=NamedRerouteUsage(), B=Constant() {r=0.7}) {attenuationRadius=0.05}
  )
  n99 = VectorParameter() {parameterName=SocksColor, defaultValue=<rgba 0.9047 0.9047 0.9047 0>}
  n100 = SphereMask(A=NamedRerouteUsage(), B=Constant() {r=0.8}) {attenuationRadius=0.05}
  n101 = SphereMask(A=NamedRerouteUsage(), B=Constant() {r=1}) {attenuationRadius=0.05}
  n102 = SphereMask(A=NamedRerouteUsage(), B=Constant() {r=1}) {attenuationRadius=0.05}
  NamedRerouteDeclaration(None=LinearInterpolate(A=LinearInterpolate(A=@n98, B=@n99.RGB, Alpha=@n100), B=NamedRerouteUsage(), Alpha=Subtract(A=@n101, B=@n102))) {name=CustomColor, nodeColor=<rgba 0.5703 1 0 1>}
  n103 = VertexColor()
  NamedRerouteDeclaration(None=@n103.R) {name=VertexColorR, nodeColor=<rgba 0 0.0391 1 1>}
  NamedRerouteDeclaration(None=@n103.G) {name=VertexColorG, nodeColor=<rgba 0.5234 0 1 1>}
  n104 = Multiply(B=Divide(B=@n3.R)) {constA=38}
  n105 = TextureSample(UVs=MaterialFunctionCall(X=Add(A=@n104, B=@n4), Y=@n7) {materialFunction=&MakeFloat2}.Result) {
   samplerType=SAMPLERTYPE_LinearColor
   texture=&TeriyakiFishToon_ColorAtlas
  }
  NamedRerouteDeclaration(None=@n105.R) {name="RGHOverride - LeftEye&&Pupil", nodeColor=<rgba 0 0.8943 1 1>}
  NamedRerouteDeclaration(None=@n105.G) {name="RGHOverride - RightEye&&Pupil", nodeColor=<rgba 0.7266 0.7738 1 1>}
  n106 = NamedRerouteUsage()
  n107 = Divide(B=@n3.R)
  n108 = Multiply(B=@n107) {constA=27, constB=0.0125}
  n109 = Constant() {r=0.01}
  n110 = NamedRerouteUsage()
  n111 = Multiply(B=@n107) {constA=32, constB=0.0125}
  n112 = Constant() {r=0.01}
  n113 = LinearInterpolate(B=Add(A=SphereMask(A=@n106, B=@n108, Radius=@n109), B=SphereMask(A=@n110, B=@n111, Radius=@n112)), Alpha=NamedRerouteUsage())
  n114 = Add(
   A=SphereMask(A=NamedRerouteUsage(), B=Multiply(B=@n107) {constA=26, constB=0.0125}, Radius=Constant() {r=0.01})
   B=SphereMask(A=NamedRerouteUsage(), B=Multiply(B=@n107) {constA=31, constB=0.0125}, Radius=Constant() {r=0.01})
  )
  NamedRerouteDeclaration(None=LinearInterpolate(A=@n113, B=Add(A=@n113, B=@n114), Alpha=NamedRerouteUsage())) {name=RGHOverride_Mask, nodeColor=<rgba 1 0 0.4453 1>}
  n115 = VertexColor()
  n116 = Divide(B=@n3.G)
  n117 = ScalarParameter() {defaultValue=0.5, parameterName=Index_Offset_Y}
  n118 = TextureSample(UVs=MaterialFunctionCall(X=Add(A=@n115, B=@n4) {constB=0.0125}, Y=Multiply(A=@n116, B=@n117)) {materialFunction=&MakeFloat2}.Result) {
   samplerType=SAMPLERTYPE_LinearColor
   texture=&TeriyakiFishToon_ColorAtlas
  }
  NamedRerouteDeclaration(None=@n118.RGB) {name=Default Color, nodeColor=<rgba 1 0 0.7031 1>}
  NamedRerouteDeclaration(None=@n48) {name=Debug, nodeColor=<rgba 1 0.6328 0 1>}
  ScalarParameter() {defaultValue=0.15, parameterName=lineThickness}
  VectorParameter() {parameterName=PupilColor, defaultValue=<rgba 0.02 0.02 0.02 0>}
 ]
}
