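# Animation Blueprint for the Teriyaki Fish Toon player body, adding secondary pouch motion and helper-bone correction to the pose copied from the attached parent mesh.
# Anim Dynamics simulates the pouch, then weighted multi-bone constraints drive the pouch base from pelvis and spine, and both knee helpers from leg bones.
# The event graph lowers the simulation's alpha during skydive float or dive, and toggles its external force after short retriggerable delays on jump and fall events.
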AnimBlueprint /BRCosmetics/Characters/Player/Male/Medium/Bodies/M_MED_Teriyaki_Fish_Toon/Meshes/M_MED_Teriyaki_Fish_Toon_AnimBP
target_skeleton=&/M_MED_Teriyaki_Fish_Toon_Skeleton
graphs [
 AnimGraph {nodes [
   n1 = Anim Dynamics - Bone: dyn_pouch_parent(
    ComponentPose=Local To Component(LocalPose=Copy Pose From Mesh() {bUseAttachedParent=true})
    Alpha=Get Skydive_Alpha() -> (Skydive_Alpha:double)
    ExternalForce=Get Impulse_Alpha() -> (Impulse_Alpha:Vector)
   ) {
    angularDampingOverride=1
    bGravityOverrideInSimSpace=true
    bOverrideAngularDamping=true
    bOverrideLinearDamping=true
    linearDampingOverride=0.9999
    simulationSpace=RootRelative
    alphaBoolBlend {blendOption=Linear}
    boundBone {boneName=dyn_pouch_parent}
    componentAppliedLinearAccClamp=<1000 1000 1000>
    componentLinearAccScale=<0.1 0.1 0.1>
    componentLinearVelScale=<0.1 0.1 0.1>
    physicsBodyDefinitions [{
      sphereCollisionRadius=0
      boundBone {boneName=dyn_pouch_parent}
      boxExtents=<16 12 6>
      constraintSetup {angularLimitsMax=<0 40 15>, angularLimitsMin=<0 -2 -15>, angularTarget=<0 0 0>}
      localJointOffset=<7 0 0>
     }]
    relativeSpaceBone {boneName=pelvis}
   }
   n2 = "Constraint - dyn_pouch_base (PPPPP)"!(
    ComponentPose=@n1
    Alpha=1
    ConstraintWeights_0=0.2879
    ConstraintWeights_1:float=0.2137
    ConstraintWeights_2:float=0.1056
    ConstraintWeights_3:float=0.0116
    ConstraintWeights_4:float=0.3812
   ) -> (Pose) {
    alphaBoolBlend {blendOption=Linear}
    boneToModify {boneName=dyn_pouch_base}
    constraintSetup [
     {targetBone {boneName=pelvis}}
     {targetBone {boneName=spine_02}}
     {targetBone {boneName=spine_03}}
     {targetBone {boneName=spine_04}}
     {targetBone {boneName=spine_01}}
    ]
    constraintWeights [1,1,1,1,1]
   }
   n3 = "Constraint - const_knee_r (PPPP)"!(
    ComponentPose=@n2
    Alpha=1
    ConstraintWeights_0=0.5711
    ConstraintWeights_1:float=0.0215
    ConstraintWeights_2:float=0.001
    ConstraintWeights_3:float=0.4064
   ) -> (Pose) {
    alphaBoolBlend {blendOption=Linear}
    boneToModify {boneName=const_knee_r}
    constraintSetup [
     {targetBone {boneName=thigh_r}}
     {targetBone {boneName=calf_twist_02_r}}
     {targetBone {boneName=thigh_twist_01_r}}
     {targetBone {boneName=calf_r}}
    ]
    constraintWeights [1,1,1,1]
   }
   n4 = "Constraint - const_knee_l (PPP)"!(ComponentPose=@n3, Alpha=1, ConstraintWeights_0=0.4248, ConstraintWeights_1:float=0.019, ConstraintWeights_2:float=0.5562) -> (Pose) {
    alphaBoolBlend {blendOption=Linear}
    boneToModify {boneName=const_knee_l}
    constraintSetup [{targetBone {boneName=calf_l}}, {targetBone {boneName=calf_twist_02_l}}, {targetBone {boneName=thigh_l}}]
    constraintWeights [1,1,1]
   }
   Output Pose(Result=Component To Local(ComponentPose=@n4)) {layerGroup=None}
  ]}
 EventGraph {nodes [
   Event BlueprintUpdateAnimation()
   n1 = Get bIsSkydiveFloating()
   n2 = Get bIsSkydiveDiveMode()
   > Set Skydive_Alpha(Skydive_Alpha:double=SelectFloat(A=0.25, B=1, bPickA=OR Boolean(A=@n1, B=@n2))) -> (Output_Get:double)
   Event OnBeginJumping()
   > RetriggerableDelay(Duration=0.2)
   > Set Impulse_Alpha(Impulse_Alpha:Vector="500,0,2000") -> (Output_Get:Vector)
   Event OnEndJumping()
   > RetriggerableDelay(Duration=0.375)
   > Set Impulse_Alpha(Impulse_Alpha="0, 0, 0")
   Event OnEndFalling()
   > RetriggerableDelay(Duration=0.375)
   > Set Impulse_Alpha(Impulse_Alpha="0, 0, 0")
   Event OnBeginFalling()
   > RetriggerableDelay(Duration=0.2)
   > Set Impulse_Alpha(Impulse_Alpha="500,0,2000")
  ]}
]
variables [
 Impulse_Alpha {type=Vector}
 Skydive_Alpha {type=number}
]
