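# Physics asset for the face accessory mesh of the male medium Teriyaki Fish Toon body, giving its hat bones secondary motion.
# Mirrored left and right chains of box bodies run from sphere bases through parent and three segments, with swing limits loosening toward the tips.
# Three top capsules each constrain directly to a top master sphere rather than chaining together.
# All joints share a tight twist limit; no constraints link bases or masters.
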PhysicsAsset /BRCosmetics/Characters/Player/Male/Medium/Bodies/M_MED_Teriyaki_Fish_Toon/Meshes/Parts/M_MED_Teriyaki_Fish_Toon_FaceAcc_Physics
bodies {
 dyn_hat_01_l [{shapeType=Box, extentX=4.792, extentY=8.3125, extentZ=1.524, shapeName=dyn_hat_01_l_box, center=<2.4579 0 0.0001>}]
 dyn_hat_01_r [{shapeType=Box, extentX=4.483, extentY=8.3715, extentZ=1.2998, shapeName=dyn_hat_01_r_box, center=<2.3654 0 0.1653>}]
 dyn_hat_02_l [{shapeType=Box, extentX=3.6772, extentY=8.0822, extentZ=1.5412, shapeName=dyn_hat_02_l_box, center=<1.7892 0 0>}]
 dyn_hat_02_r [{shapeType=Box, extentX=3.7432, extentY=8.0675, extentZ=1.3307, shapeName=dyn_hat_02_r_box, center=<1.8635 0 0.1363>}]
 dyn_hat_03_l [{shapeType=Box, extentX=3.1961, extentY=7.3675, extentZ=1.4626, shapeName=dyn_hat_03_l_box, center=<1.3285 0 0.0001>}]
 dyn_hat_03_r [{shapeType=Box, extentX=3.0559, extentY=6.7374, extentZ=1.5412, shapeName=dyn_hat_03_r_box, center=<1.4423 0 0>}]
 dyn_hat_base_l [{shapeType=Sphere, radius=0.9645, shapeName=dyn_hat_base_l_sphere}]
 dyn_hat_base_r [{shapeType=Sphere, radius=0.8411, shapeName=dyn_hat_base_r_sphere}]
 dyn_hat_master [{shapeType=Sphere, radius=0.9645, shapeName=dyn_hat_master_sphere}]
 dyn_hat_parent_l [{
   shapeType=Box
   extentX=5.6411
   extentY=8.3768
   extentZ=1.3296
   shapeName=dyn_hat_parent_l_box
   center=<2.6063 -0.0074 -0.132>
   rotation=<rot -0.1125 2.0135 3.1975>
  }]
 dyn_hat_parent_r [{shapeType=Box, extentX=5.7432, extentY=7.9813, extentZ=1.5812, shapeName=dyn_hat_parent_r_box, center=<2.7076 0 0>}]
 dyn_hat_top_01 [{shapeType=Capsule, length=1.5304, radius=2.1121, shapeName=dyn_hat_top_01_capsule, center=<2.1454 -0.0003 0>, rotation=<rot 90 0 0>}]
 dyn_hat_top_02 [{shapeType=Capsule, length=0.8392, radius=2.4272, shapeName=dyn_hat_top_02_capsule, center=<2.1454 -0.0003 0>, rotation=<rot 90 0 0>}]
 dyn_hat_top_03 [{shapeType=Capsule, length=2.9642, radius=2.0317, shapeName=dyn_hat_top_03_capsule, center=<2.1454 -0.0003 0>, rotation=<rot 90 0 0>}]
 dyn_hat_top_master [{shapeType=Sphere, radius=0.9645, shapeName=dyn_hat_top_master_sphere}]
}
constraints [
 {bone1Name=dyn_hat_top_01, bone2Name=dyn_hat_top_master, swing1LimitDegrees=10, swing2LimitDegrees=10, twistLimitDegrees=3}
 {bone1Name=dyn_hat_top_02, bone2Name=dyn_hat_top_master, swing1LimitDegrees=10, swing2LimitDegrees=10, twistLimitDegrees=3}
 {bone1Name=dyn_hat_top_03, bone2Name=dyn_hat_top_master, swing1LimitDegrees=10, swing2LimitDegrees=10, twistLimitDegrees=3}
 {bone1Name=dyn_hat_parent_r, bone2Name=dyn_hat_base_r, swing1LimitDegrees=10, swing2LimitDegrees=10, twistLimitDegrees=3}
 {bone1Name=dyn_hat_01_r, bone2Name=dyn_hat_parent_r, swing1LimitDegrees=10, swing2LimitDegrees=15, twistLimitDegrees=3}
 {bone1Name=dyn_hat_02_r, bone2Name=dyn_hat_01_r, swing1LimitDegrees=20, swing2LimitDegrees=20, twistLimitDegrees=3}
 {bone1Name=dyn_hat_03_r, bone2Name=dyn_hat_02_r, swing1LimitDegrees=20, swing2LimitDegrees=20, twistLimitDegrees=3}
 {bone1Name=dyn_hat_parent_l, bone2Name=dyn_hat_base_l, swing1LimitDegrees=10, swing2LimitDegrees=10, twistLimitDegrees=3}
 {bone1Name=dyn_hat_01_l, bone2Name=dyn_hat_parent_l, swing1LimitDegrees=10, swing2LimitDegrees=15, twistLimitDegrees=3}
 {bone1Name=dyn_hat_02_l, bone2Name=dyn_hat_01_l, swing1LimitDegrees=20, swing2LimitDegrees=20, twistLimitDegrees=3}
 {bone1Name=dyn_hat_03_l, bone2Name=dyn_hat_02_l, swing1LimitDegrees=20, swing2LimitDegrees=20, twistLimitDegrees=3}
]
