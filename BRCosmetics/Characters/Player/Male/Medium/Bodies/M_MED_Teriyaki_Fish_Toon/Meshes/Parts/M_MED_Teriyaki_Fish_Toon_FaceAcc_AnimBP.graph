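# Animation Blueprint that adds physics secondary motion to the face accessory of the Teriyaki Fish Toon player body.
# It copies the attached parent mesh's pose, then runs a RigidBody simulation in spine_05 base-bone space with damped component motion and overridden gravity.
# Jump and fall events use retriggerable delays to set Impulse_Alpha, a forward-and-up force vector fed into the simulation as external force, and to zero it when they end.
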AnimBlueprint /BRCosmetics/Characters/Player/Male/Medium/Bodies/M_MED_Teriyaki_Fish_Toon/Meshes/Parts/M_MED_Teriyaki_Fish_Toon_FaceAcc_AnimBP
target_skeleton=&/M_MED_Teriyaki_Fish_Toon_FaceAcc_Skeleton
graphs [
 AnimGraph {nodes [
   n1 = RigidBody(
    ComponentPose=Local To Component(LocalPose=Copy Pose From Mesh() {bUseAttachedParent=true})
    ExternalForce=Get Impulse_Alpha() -> (Impulse_Alpha:Vector)
   ) {
    bClampLinearTranslationLimitToRefPose=true
    bOverrideWorldGravity=true
    overridePhysicsAsset=&/M_MED_Teriyaki_Fish_Toon_FaceAcc_Physics
    simulationSpace=BaseBoneSpace
    alphaBoolBlend {blendOption=Linear}
    baseBoneRef {boneName=spine_05}
    componentAppliedLinearAccClamp=<0.25 0.25 0.25>
    componentLinearAccScale=<0.25 0.25 0.25>
    componentLinearVelScale=<0.25 0.25 0.25>
    overrideWorldGravity {z=-980}
   }
   Output Pose(Result=Component To Local(ComponentPose=@n1)) {layerGroup=None}
  ]}
 EventGraph {nodes [
   Event OnBeginJumping()
   > RetriggerableDelay(Duration=0.25)
   > Set Impulse_Alpha(Impulse_Alpha:Vector="2500,0,5000") -> (Output_Get:Vector)
   Event OnEndJumping()
   > RetriggerableDelay(Duration=0.4)
   > Set Impulse_Alpha(Impulse_Alpha="0, 0, 0")
   Event OnEndFalling()
   > RetriggerableDelay(Duration=0.4)
   > Set Impulse_Alpha(Impulse_Alpha="0, 0, 0")
   Event OnBeginFalling()
   > RetriggerableDelay(Duration=0.25)
   > Set Impulse_Alpha(Impulse_Alpha="2500,0,5000")
   Event BlueprintUpdateAnimation()
  ]}
]
variables [Impulse_Alpha {type=Vector}]
